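# Widget preview of the autorun lock zone for the HUD Layout Tool V2, derived from HUDLayoutToolV2_AutorunPreview.
# A square SizeBox holds an overlay that stacks a translucent button background, a centered autorun icon and a bottom-right instances-left counter.
# The overlay and counter fill inherited widget slots, while the inherited Border_Background is left empty.
# The event graph is empty, so this asset only defines layout and styling.
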WidgetBlueprint /Game/Athena/HUD/Bacchus/HUDLayoutTool/WidgetPreview/HUDLayoutToolV2_AutorunLockZonePreview
graphs [EventGraph {links [], nodes []}]
variables []
widgets {
 info {inheritedWidgetCount=3, parentClass=/Script/FortniteUI.HUDLayoutToolV2_AutorunPreview, rootWidgetClass=&SizeBox, widgetCount=6}
 widgets [
  SizeBox {
   widgetClassPath=&SizeBox
   widget=SizeBox
   properties {bOverride_MaxAspectRatio=true, bOverride_MinAspectRatio=true, heightOverride=200, visibility=Visible, widthOverride=200}
  }
  Overlay_AutorunLockZone {widgetClassPath=&Overlay, bInherited=true, parent=SizeBox, widget=Overlay_AutorunLockZone, slot {class=SizeBoxSlot}}
  Background {
   widgetClassPath=&Image
   parent=Overlay_AutorunLockZone
   widget=Background
   properties {
    brush {
     bIsDynamicallyLoaded=false
     drawAs=Image
     imageType=FullColor
     mirroring=NoMirror
     resourceName=None
     resourceObject=/Game/Athena/HUD/Bacchus/Material/MI_UI_ButtonBG_Light
     tiling=NoTile
     imageSize=<128 128>
     margin=<ltrb 0 0 0 0>
     outlineSettings {
      bUseBrushTransparency=false
      roundingType=HalfHeightRadius
      width=0
      color {colorUseRule=UseColor_Specified, specifiedColor=<rgba 0 0 0 0>}
      cornerRadii=<0 0 0 0>
     }
     tintColor {colorUseRule=UseColor_Specified, specifiedColor=<rgba 1 1 1 1>}
     uVRegion {bIsValid=false, max=<0 0>, min=<0 0>}
    }
    colorAndOpacity=<rgba 1 1 1 0.6>
   }
   slot {class=OverlaySlot, properties {horizontalAlignment=HAlign_Fill, verticalAlignment=VAlign_Fill}}
  }
  Icon {
   widgetClassPath=&Image
   parent=Overlay_AutorunLockZone
   widget=Icon
   properties {
    visibility=HitTestInvisible
    brush {
     bIsDynamicallyLoaded=false
     drawAs=Image
     imageType=NoImage
     mirroring=NoMirror
     resourceName=None
     resourceObject=/Game/Athena/HUD/Atlas/Bacchus/Autorun_Normal_Sprite
     tiling=NoTile
     imageSize=<80 80>
     margin=<ltrb 0 0 0 0>
     outlineSettings {
      bUseBrushTransparency=false
      roundingType=HalfHeightRadius
      width=0
      color {colorUseRule=UseColor_Specified, specifiedColor=<rgba 0 0 0 0>}
      cornerRadii=<0 0 0 0>
     }
     tintColor {colorUseRule=UseColor_Specified, specifiedColor=<rgba 1 1 1 1>}
     uVRegion {bIsValid=false, max=<0 0>, min=<0 0>}
    }
    colorAndOpacity=<rgba 1 1 1 0.6>
   }
   slot {class=OverlaySlot, properties {horizontalAlignment=HAlign_Center, verticalAlignment=VAlign_Center}}
  }
  TextBlock_InstancesLeftCounter {
   widgetClassPath=&CommonTextBlock
   bInherited=true
   parent=Overlay_AutorunLockZone
   widget=TextBlock_InstancesLeftCounter
   properties {
    justification=Center
    scrollStyle=&TextScrollStyle-Base_C
    textTransformPolicy=ToUpper
    font {
     bForceMonospaced=false
     bMaterialIsStencil=false
     fontMaterial=null
     fontObject=&BurbankSmall
     letterSpacing=0
     monospacedWidth=1
     size=16
     skewAmount=0
     typefaceFontName=Black
     outlineSettings {
      bApplyOutlineToDropShadows=false
      bMiteredCorners=false
      bSeparateFillAlpha=false
      outlineMaterial=null
      outlineSize=0
      outlineColor=<rgba 0 0 0 1>
     }
    }
    shadowOffset=<0 0>
   }
   slot {class=OverlaySlot, properties {horizontalAlignment=HAlign_Right, verticalAlignment=VAlign_Bottom}}
  }
  Border_Background {widgetClassPath=&Border, bInherited=true, widget=null, properties=~}
 ]
}
