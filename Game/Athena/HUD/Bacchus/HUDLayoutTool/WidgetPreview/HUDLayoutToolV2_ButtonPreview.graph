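# Widget Blueprint for a button preview in the HUD Layout Tool V2, based on the native HUDLayoutToolV2_ButtonPreview class.
# An outer hitbox SizeBox holds a centered visual SizeBox, whose ScaleBox scales a fixed-size background and translucent icon, keeping hit and visual sizes separate.
# An instances-left counter sits bottom-right, outside the scaled content.
# The event graph is empty, and the inherited Border_Background is absent from the tree.
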WidgetBlueprint /Game/Athena/HUD/Bacchus/HUDLayoutTool/WidgetPreview/HUDLayoutToolV2_ButtonPreview
graphs [EventGraph {links [], nodes []}]
variables []
widgets {
 info {inheritedWidgetCount=3, parentClass=/Script/FortniteUI.HUDLayoutToolV2_ButtonPreview, rootWidgetClass=&SizeBox, widgetCount=11}
 widgets [
  SizeBox_HitboxSize {
   widgetClassPath=&SizeBox
   widget=SizeBox_HitboxSize
   properties {
    bOverride_HeightOverride=true
    bOverride_MaxAspectRatio=true
    bOverride_MinAspectRatio=true
    bOverride_WidthOverride=true
    heightOverride=128
    visibility=Visible
    widthOverride=128
   }
  }
  Overlay_2 {widgetClassPath=&Overlay, parent=SizeBox_HitboxSize, widget=Overlay_2, slot {class=SizeBoxSlot}}
  SizeBox_VisualSize {
   widgetClassPath=&SizeBox
   parent=Overlay_2
   widget=SizeBox_VisualSize
   properties {bOverride_HeightOverride=true, bOverride_WidthOverride=true, heightOverride=128, widthOverride=128}
   slot {class=OverlaySlot, properties {horizontalAlignment=HAlign_Center, verticalAlignment=VAlign_Center}}
  }
  ScaleBox_0 {widgetClassPath=&ScaleBox, parent=SizeBox_VisualSize, widget=ScaleBox_0, slot {class=SizeBoxSlot}}
  SizeBox_0 {
   widgetClassPath=&SizeBox
   parent=ScaleBox_0
   widget=SizeBox_0
   properties {bOverride_HeightOverride=true, bOverride_WidthOverride=true, heightOverride=128, widthOverride=128}
   slot {class=ScaleBoxSlot}
  }
  Overlay_0 {widgetClassPath=&Overlay, parent=SizeBox_0, widget=Overlay_0, slot {class=SizeBoxSlot}}
  Image_Background {
   widgetClassPath=&Image
   parent=Overlay_0
   widget=Image_Background
   properties {brush {
     bIsDynamicallyLoaded=false
     drawAs=Image
     imageType=FullColor
     mirroring=NoMirror
     resourceName=None
     resourceObject=/Game/Athena/HUD/Bacchus/Material/MI_UI_ButtonBG_Light
     tiling=NoTile
     imageSize=<128 128>
     margin=<ltrb 0 0 0 0>
     outlineSettings {
      bUseBrushTransparency=false
      roundingType=HalfHeightRadius
      width=0
      color {colorUseRule=UseColor_Specified, specifiedColor=<rgba 0 0 0 0>}
      cornerRadii=<0 0 0 0>
     }
     tintColor {colorUseRule=UseColor_Specified, specifiedColor=<rgba 1 1 1 1>}
     uVRegion {bIsValid=false, max=<0 0>, min=<0 0>}
    }}
   slot {class=OverlaySlot, properties {horizontalAlignment=HAlign_Fill, verticalAlignment=VAlign_Fill}}
  }
  SizeBox {
   widgetClassPath=&SizeBox
   parent=Overlay_0
   widget=SizeBox
   properties {
    bOverride_MaxDesiredHeight=true
    bOverride_MaxDesiredWidth=true
    bOverride_MinDesiredHeight=true
    bOverride_MinDesiredWidth=true
    heightOverride=128
    maxDesiredHeight=128
    maxDesiredWidth=128
    minDesiredHeight=60
    minDesiredWidth=60
    widthOverride=128
   }
   slot {class=OverlaySlot, properties {horizontalAlignment=HAlign_Center, verticalAlignment=VAlign_Center}}
  }
  Image_Icon {
   widgetClassPath=&Image
   bInherited=true
   parent=SizeBox
   widget=Image_Icon
   properties {
    visibility=HitTestInvisible
    brush {
     bIsDynamicallyLoaded=false
     drawAs=Image
     imageType=FullColor
     mirroring=NoMirror
     resourceName=None
     resourceObject=/Game/Athena/HUD/Art/BacchusHudIcons/Confirm_Normal_L
     tiling=NoTile
     imageSize=<72 72>
     margin=<ltrb 0 0 0 0>
     outlineSettings {
      bUseBrushTransparency=false
      roundingType=HalfHeightRadius
      width=0
      color {colorUseRule=UseColor_Specified, specifiedColor=<rgba 0 0 0 0>}
      cornerRadii=<0 0 0 0>
     }
     tintColor {colorUseRule=UseColor_Specified, specifiedColor=<rgba 1 1 1 1>}
     uVRegion {bIsValid=false, max=<0 0>, min=<0 0>}
    }
    colorAndOpacity=<rgba 1 1 1 0.6>
   }
   slot {class=SizeBoxSlot}
  }
  TextBlock_InstancesLeftCounter {
   widgetClassPath=&CommonTextBlock
   bInherited=true
   parent=Overlay_2
   widget=TextBlock_InstancesLeftCounter
   properties {
    justification=Center
    textTransformPolicy=ToUpper
    font {
     bForceMonospaced=false
     bMaterialIsStencil=false
     fontMaterial=null
     fontObject=&BurbankSmall
     letterSpacing=0
     monospacedWidth=1
     size=16
     skewAmount=0
     typefaceFontName=Black
     outlineSettings {
      bApplyOutlineToDropShadows=false
      bMiteredCorners=false
      bSeparateFillAlpha=false
      outlineMaterial=null
      outlineSize=0
      outlineColor=<rgba 0 0 0 1>
     }
    }
    shadowOffset=<0 0>
   }
   slot {class=OverlaySlot, properties {horizontalAlignment=HAlign_Right, verticalAlignment=VAlign_Bottom}}
  }
  Border_Background {widgetClassPath=&Border, bInherited=true, widget=null, properties=~}
 ]
}
