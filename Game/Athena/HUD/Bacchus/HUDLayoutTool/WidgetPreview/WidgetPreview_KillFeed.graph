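# Widget Blueprint preview of the kill feed for the HUD Layout Tool V2, derived from HUDLayoutToolV2_WidgetPreview.
# A fixed-height SizeBox holds the inherited translucent background border, inside which a slightly scaled-down stack shows three identical sample elimination messages.
# The messages are rich text using Bacchus kill feed styles and an image decorator.
# The inherited instances-left counter sits bottom-right, and the event graph is empty.
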WidgetBlueprint /Game/Athena/HUD/Bacchus/HUDLayoutTool/WidgetPreview/WidgetPreview_KillFeed
graphs [EventGraph {links [], nodes []}]
variables []
widgets {
 info {inheritedWidgetCount=2, parentClass=/Script/FortniteUI.HUDLayoutToolV2_WidgetPreview, rootWidgetClass=&SizeBox, widgetCount=10}
 widgets [
  SizeBox_15 {widgetClassPath=&SizeBox, widget=SizeBox_15, properties {bOverride_HeightOverride=true, heightOverride=128}}
  Border_Background {
   widgetClassPath=&Border
   bInherited=true
   parent=SizeBox_15
   widget=Border_Background
   properties {brushColor=<rgba 0 0 0 0.2>, padding=<ltrb 0 0 0 0>}
   slot {class=SizeBoxSlot}
  }
  Overlay_0 {
   widgetClassPath=&Overlay
   parent=Border_Background
   widget=Overlay_0
   slot {class=BorderSlot, properties {padding=<ltrb 0 0 0 0>}}
  }
  ScaleBox_0 {
   widgetClassPath=&ScaleBox
   parent=Overlay_0
   widget=ScaleBox_0
   properties {stretch=UserSpecified, userSpecifiedScale=0.9}
   slot {class=OverlaySlot, properties {horizontalAlignment=HAlign_Fill, verticalAlignment=VAlign_Fill, padding=<ltrb 20 0 20 0>}}
  }
  Overlay {
   widgetClassPath=&Overlay
   parent=ScaleBox_0
   widget=Overlay
   properties {renderTransform {angle=0, scale=<1 1>, shear=<0 0>, translation=<0 -2>}}
   slot {class=ScaleBoxSlot}
  }
  VerticalBox {
   widgetClassPath=&VerticalBox
   parent=Overlay
   widget=VerticalBox
   properties {visibility=Visible}
   slot {class=OverlaySlot, properties {horizontalAlignment=HAlign_Fill, verticalAlignment=VAlign_Fill}}
  }
  KillText_1 {
   widgetClassPath=/Script/UMG.RichTextBlock
   parent=VerticalBox
   widget=KillText_1
   properties {
    text="<LocalKiller>You</> eliminated <Victim>Epic Games</>"
    textStyleSet=/Game/Athena/HUD/Bacchus/BacchusKillFeedStyles
    decoratorClasses [/Game/Athena/HUD/TextStyles/KillFeedImageDecorator.KillFeedImageDecorator_C]
   }
   slot {class=VerticalBoxSlot}
  }
  KillText2_1 {
   widgetClassPath=&RichTextBlock
   parent=VerticalBox
   widget=KillText2_1
   properties {
    text="<LocalKiller>You</> eliminated <Victim>Epic Games</>"
    textStyleSet=&BacchusKillFeedStyles
    decoratorClasses [&KillFeedImageDecorator_C]
   }
   slot {class=VerticalBoxSlot}
  }
  KillText3_1 {
   widgetClassPath=&RichTextBlock
   parent=VerticalBox
   widget=KillText3_1
   properties {
    text="<LocalKiller>You</> eliminated <Victim>Epic Games</>"
    textStyleSet=&BacchusKillFeedStyles
    decoratorClasses [&KillFeedImageDecorator_C]
   }
   slot {class=VerticalBoxSlot}
  }
  TextBlock_InstancesLeftCounter {
   widgetClassPath=&CommonTextBlock
   bInherited=true
   parent=Overlay_0
   widget=TextBlock_InstancesLeftCounter
   properties {
    justification=Center
    textTransformPolicy=ToUpper
    font {
     bForceMonospaced=false
     bMaterialIsStencil=false
     fontMaterial=null
     fontObject=&BurbankSmall
     letterSpacing=0
     monospacedWidth=1
     size=16
     skewAmount=0
     typefaceFontName=Black
     outlineSettings {
      bApplyOutlineToDropShadows=false
      bMiteredCorners=false
      bSeparateFillAlpha=false
      outlineMaterial=null
      outlineSize=0
      outlineColor=<rgba 0 0 0 1>
     }
    }
    shadowOffset=<0 0>
   }
   slot {class=OverlaySlot, properties {horizontalAlignment=HAlign_Right, verticalAlignment=VAlign_Bottom, padding=<ltrb 0 0 4 0>}}
  }
 ]
}
