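# Widget Blueprint preview tile for the material (resource) selector element in Fortnite's mobile HUD layout tool, built on HUDLayoutToolV2_WidgetPreview.
# A height-fixed SizeBox wraps an Overlay holding the inherited translucent background Border, which centers a static resource-selector placeholder image.
# The inherited instances-left counter text sits in the bottom-right corner.
# It has no graph logic or variables, so any behavior comes from the parent class.
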WidgetBlueprint /Game/Athena/HUD/Bacchus/HUDLayoutTool/WidgetPreview/WidgetPreview_MaterialSelector
graphs [EventGraph {links [], nodes []}]
variables []
widgets {
 info {inheritedWidgetCount=2, parentClass=/Script/FortniteUI.HUDLayoutToolV2_WidgetPreview, rootWidgetClass=&SizeBox, widgetCount=5}
 widgets [
  SizeBox_15 {widgetClassPath=&SizeBox, widget=SizeBox_15, properties {bOverride_HeightOverride=true, heightOverride=128}}
  Overlay_0 {widgetClassPath=&Overlay, parent=SizeBox_15, widget=Overlay_0, slot {class=SizeBoxSlot}}
  Border_Background {
   widgetClassPath=&Border
   bInherited=true
   parent=Overlay_0
   widget=Border_Background
   properties {horizontalAlignment=HAlign_Center, verticalAlignment=VAlign_Center, brushColor=<rgba 0 0 0 0.2>, padding=<ltrb 20 0 20 0>}
   slot {class=OverlaySlot, properties {horizontalAlignment=HAlign_Fill, verticalAlignment=VAlign_Fill}}
  }
  PlaceHolder {
   widgetClassPath=&Image
   parent=Border_Background
   widget=PlaceHolder
   properties {
    brush {
     bIsDynamicallyLoaded=false
     drawAs=Image
     imageType=NoImage
     mirroring=NoMirror
     resourceName=None
     resourceObject=/Game/Athena/HUD/Bacchus/HUDLayoutTool/WidgetPreview/WidgetPreview_PlaceHolders/T-HLT-RessourceSelector
     tiling=NoTile
     imageSize=<185 83>
     margin=<ltrb 0 0 0 0>
     outlineSettings {
      bUseBrushTransparency=false
      roundingType=HalfHeightRadius
      width=0
      color {colorUseRule=UseColor_Specified, specifiedColor=<rgba 0 0 0 0>}
      cornerRadii=<0 0 0 0>
     }
     tintColor {colorUseRule=UseColor_Specified, specifiedColor=<rgba 1 1 1 1>}
     uVRegion {bIsValid=false, max=<0 0>, min=<0 0>}
    }
    renderTransform {angle=0, scale=<1 1>, shear=<0 0>, translation=<0 -2>}
   }
   slot {class=BorderSlot, properties {horizontalAlignment=HAlign_Center, verticalAlignment=VAlign_Center, padding=<ltrb 20 0 20 0>}}
  }
  TextBlock_InstancesLeftCounter {
   widgetClassPath=&CommonTextBlock
   bInherited=true
   parent=Overlay_0
   widget=TextBlock_InstancesLeftCounter
   properties {
    justification=Center
    textTransformPolicy=ToUpper
    font {
     bForceMonospaced=false
     bMaterialIsStencil=false
     fontMaterial=null
     fontObject=&BurbankSmall
     letterSpacing=0
     monospacedWidth=1
     size=16
     skewAmount=0
     typefaceFontName=Black
     outlineSettings {
      bApplyOutlineToDropShadows=false
      bMiteredCorners=false
      bSeparateFillAlpha=false
      outlineMaterial=null
      outlineSize=0
      outlineColor=<rgba 0 0 0 1>
     }
    }
    shadowOffset=<0 0>
   }
   slot {class=OverlaySlot, properties {horizontalAlignment=HAlign_Right, verticalAlignment=VAlign_Bottom, padding=<ltrb 0 0 4 0>}}
  }
 ]
}
